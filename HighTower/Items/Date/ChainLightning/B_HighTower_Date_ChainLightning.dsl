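; Dual-hand ranged weapon for the HighTower Date chain lightning gauntlets, handling firing animation, force feedback and muzzle placement.
; On the local client, trigger presses and a fire-rate timer start firing animations after ability, overheat, CanFire and held-input checks (AI skips held input).
; The server multicasts the animation to remote clients through the gauntlets' custom character part.
; Muzzle location follows the last fired hand on the gauntlet mesh, with boss and pawn-mesh fallbacks.
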(event EventBeginPlay
  (bind _returnvalue (Game|GetInstigator))
  ;[0.1111 1 0 0.5] BeginPlay hook up listeners
  (|CallParentFunction)
  (Utilities|IsValid _returnvalue
    (:Is%20Valid
      (if (Pawn|IsLocallyControlled _returnvalue)
        (EventDispatchers|BindEventtoOnPressTrigger (EventDispatchers|CreateEvent :!event Custom|OnTriggerPressed))
        (RangedWeaponDual|BindEventtoOnHandFired (EventDispatchers|CreateEvent :!event Custom|HandFired))))))

(event EventEndPlay
  (params (EndPlayReason :type EEndPlayReason%20Enum :default "Destroyed"))
  ;[1 0.5794 0 0.5] EndPlay cleanup
  (|CallParentFunction EndPlayReason)
  (Utilities|Time|ClearandInvalidateTimerbyHandle (Variables|Default|GetRefireTimerHandle))
  (EventDispatchers|UnbindEventfromOnPressTrigger (EventDispatchers|CreateEvent :!event Custom|OnTriggerPressed))
  (RangedWeaponDual|UnbindEventfromOnHandFired (EventDispatchers|CreateEvent :!event Custom|HandFired)))

(event Custom|SetCustomCharacterPart
  (params (New%20Custom%20Character%20Part :type CCPM%20High%20Tower%20Date%20Gauntlets%20Object%20Reference))
  ;[0.3846 0.4252 1 0.5] SetCustomCharacterPart(). The Custom Character Part will be used to trigger firing animations.
  (Variables|Default|SetCustomCharacterPart New%20Custom%20Character%20Part))

(event Custom|ClearCustomCharacterPart
  (params (Custom%20Character%20Part%20To%20Remove :type CCPM%20High%20Tower%20Date%20Gauntlets%20Object%20Reference))
  ;[0.8698 0.6058 0.009776 0.5] ClearCustomCharacterPart(). If Custom Character Part and the input match, clear Custom Character Part
  (if (== Custom%20Character%20Part%20To%20Remove (Variables|Default|GetCustomCharacterPart))
    (Variables|Default|SetCustomCharacterPart 0)))

(event Custom|HandFired
  (params (Hand :type EDualWeaponHand%20Enum :default "LEFT") (bPersistantFire :type Boolean))
  ;[1 1 1] On hand fired (left - right)
  ;;[1 1 1] Comment
  (FortWeapon|PlayForceFeedback false :!label _label)
  ;;[1 0 0.04289] TEMP!!! this is to solve  handfired being called even if shoot ability is blocked
  (Utilities|FlowControl|Gate true
    (:Exit
      (goto _label))))

(event EventOnPlayWeaponFireFX
  (params (bPersistentFire :type Boolean) (bSecondaryFire :type Boolean))
  (Utilities|FlowControl|Sequence
    (:then_0
      ;[1 1 1] On hand fired (left - right)
      ;;[1 0 0.04289] TEMP!!! this is to solve  handfired being called even if shoot ability is blocked
      (Utilities|FlowControl|Gate true :!label _label
        (:Exit
          ;;[1 1 1] Comment
          (FortWeapon|PlayForceFeedback false)
          (goto _label Close))))
    (:then_1
      (FortWeapon|PlayForceFeedback false bPersistentFire))))

(event EventOnStopWeaponFireFX
  (FortWeapon|StopForceFeedback))

(event FortWeapon|EventOnEquippedAndAttached
  (bind _instigator (Variables|Actor|GetInstigator))
  ;[1 1 1] added bool
  (Variables|Default|SetIsDate true)
  (Utilities|IsValid _instigator
    (:Is%20Valid
      (if (Pawn|IsLocallyControlled _instigator)
        _))))

(event Custom|OnTriggerPressed
  (params (Weapon :type Fort%20Weapon%20Object%20Reference))
  ;[0 0.009266 1 0.5] OnTriggerPressed()
  (bind _returnvalue (CallFunction|ShouldStartFiringAnimation))
  (if _returnvalue
    (CallFunction|StartFiringAnimation) ; Play the firing animation locally
    (CallFunction|ServerStartFireAnimation) ; Tell the server to start the animation so it can tell all other clients to start the firing animation
    (bind _returnvalue_2 (Utilities|Time|SetTimerbyEvent (EventDispatchers|CreateEvent :!event Custom|RefireTimerExpired) (Math|Float|SafeDivide 1.0 (FortWeapon|GetFiringRate))))
    (Variables|Default|SetRefireTimerHandle _returnvalue_2)))

(event FortWeapon|EventOnLocalUnEquip
  ;[1 1 1] added bool
  (Variables|Default|SetIsDate false))

(event Custom|RefireTimerExpired
  ;[0 0.009266 1 0.5] OnTriggerPressed()
  (Utilities|Time|ClearandInvalidateTimerbyHandle (Variables|Default|GetRefireTimerHandle))
  (bind _returnvalue (CallFunction|ShouldStartFiringAnimation))
  (if _returnvalue
    (CallFunction|StartFiringAnimation) ; Play the firing animation locally
    (CallFunction|ServerStartFireAnimation) ; Tell the server to start the animation so it can tell all other clients to start the firing animation
    (bind _returnvalue_2 (Utilities|Time|SetTimerbyEvent (EventDispatchers|CreateEvent :!event Custom|RefireTimerExpired) (Math|Float|SafeDivide 1.0 (FortWeapon|GetFiringRate))))
    (Variables|Default|SetRefireTimerHandle _returnvalue_2)))

(event Custom|Server_StartFireAnimation
  ;[0 0.9274 1 0.5] Server_StartFireAnimation()
  ; Multicast_StartFireAnimationForRemotePlayers()
  (CallFunction|MulticastStartFireAnimationforRemotePlayers))

(event Custom|Multicast_StartFireAnimationForRemotePlayers
  (bind _returnvalue (Game|GetInstigator))
  ;[0 0.9274 1 0.5] Server_StartFireAnimation()
  ; Multicast_StartFireAnimationForRemotePlayers()
  (Networking|CanExecuteCosmeticEvents
    (:True
      ;;[0.7263 0.1833 0.9323 0.5] Have all remote clients play the firing animation (the local client should have already handled this.)
      (Utilities|IsValid _returnvalue
        (:Is%20Valid
          (if (not (Pawn|IsLocallyControlled _returnvalue))
            (CallFunction|StartFiringAnimation)))))))

(event Custom|StartFiringAnimation
  ;[0.5771 0.1143 1 0.5] StartFiringAnimation()
  ; In the base case, have CustomCharacterPart play the animation.
  (Networking|CanExecuteCosmeticEvents
    (:True
      (bind _customcharacterpart (Variables|Default|GetCustomCharacterPart)
        (:then
          (Class|CCPMHighTowerDateGauntlets|StartFiringAnimation _customcharacterpart (RangedWeaponDual|GetLastFireHand)))
        (:else)))))

(fn GetGauntletSkelMesh
  (bind _asfort_player_pawn (Utilities|Casting|CastToFortPlayerPawn (Game|GetInstigator))
    (:then
      (Utilities|IsValid (FortPawn|GetModifierforPartType _asfort_player_pawn "Gameplay")
        (:Is%20Valid
          (return (CustomCharacter|GetSkeletalMeshForPartOnPlayerPawn (FortPawn|GetModifierforPartType _asfort_player_pawn "Gameplay"))))
        (:Is%20Not%20Valid
          (return 0))))
    (:CastFailed
      (return 0))))

(fn GetMuzzleLocation
  (params (PatternIndex :type Integer))
  (bind _mesh (CallFunction|GetGauntletSkelMesh))
  (bind _instigator (Variables|Actor|GetInstigator))
  (Utilities|IsValid _mesh
    (:Is%20Valid
      (return (Transformation|GetSocketLocation _mesh (Utilities|Select :Index (RangedWeaponDual|GetLastFireHand) :LEFT "ring_metacarpal_l" :RIGHT "ring_metacarpal_r"))))
    (:Is%20Not%20Valid
      (bind _asbp_player_pawn_non_participant_boss (Utilities|Casting|CastToBP_PlayerPawn_NonParticipant_Boss _instigator)
        (:then
          (return (Transformation|GetSocketLocation (Variables|/Script/Engine.Character|GetMesh _asbp_player_pawn_non_participant_boss) (Utilities|Select :Index (RangedWeaponDual|GetLastFireHand) :LEFT "wrist_pickaxe_l" :RIGHT "wrist_pickaxe_r"))))
        (:CastFailed
          (bind _asfort_player_pawn (Utilities|Casting|CastToFortPlayerPawn _instigator)
            (:then
              (return (Transformation|GetSocketLocation (Variables|/Script/Engine.Character|GetMesh _asfort_player_pawn) (Utilities|Select :Index (RangedWeaponDual|GetLastFireHand) :LEFT "hand_pickaxe_l" :RIGHT "hand_pickaxe_r"))))
            (:CastFailed
              (return (|CallParentFunction PatternIndex)))))) ; Check if we are the AI trying to use the Gauntlet
    )))

(fn HideReticle
  (bind _asfort_player_controller_zone (Utilities|Casting|CastToFortPlayerControllerZone (Game|GetInstigatorController)))
  (HUD|SetHUDElementVisibilitywithContext _asfort_player_controller_zone (Variables|Scope|GetReticleHUDElementTags) true "None" ""))

(fn ShouldStartFiringAnimation
  (bind _asfort_player_pawn (Utilities|Casting|CastToFortPlayerPawn (Game|GetInstigator)))
  ; First pass:
  ; We can start the firing animation if all are true:
  ; 1) it has been long enough since we last fired
  ; 2) we are not using any other abilities
  ; 3) the weapon is not overheated, and
  ; 4) we are allowed to fire
  (if (Math|Boolean|ANDBoolean (not (Utilities|Time|IsTimerActivebyHandle (Variables|Default|GetRefireTimerHandle))) (not (GameplayTags|HasMatchingGameplayTag _asfort_player_pawn "(TagName=\"Granted.Athena.Power.ActivePower.IsActive\")")) (not (Overheat|IsOverheated)) (FortWeapon|CanFire true))
    (Utilities|Casting|CastToAIController (Pawn|GetController _asfort_player_pawn)
      (:then
        ;[1 0.4688 0.4688 0.5] Return that AIs can start the fire animation.
        (return true))
      (:CastFailed
        ;[1 0.4688 0.4688 0.5] Return Players can start the fire animation IF they are holding primary input.
        (return (Variables|/Script/FortniteGame.FortPawn|GetbPrimaryInputHeld _asfort_player_pawn))))
    (else
      ;[1 0.4688 0.4688 0.5] Return that we should NOT start the fire animation.
      (return false))))

(fn ShowReticle
  (bind _asfort_player_controller_zone (Utilities|Casting|CastToFortPlayerControllerZone (Game|GetInstigatorController)))
  (HUD|ResetHUDElementVisibilitytoSettings _asfort_player_controller_zone (Variables|Scope|GetReticleHUDElementTags)))

(fn ConstructionScript)
