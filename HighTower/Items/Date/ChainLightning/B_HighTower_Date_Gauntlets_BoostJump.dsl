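; Grenade-based explosion actor for the boost jump of the HighTower Date chain lightning gauntlets.
; On BeginPlay the server triggers the explosion after a zero-length delay; remote clients only cast the instigator.
; On explosion, a locally controlled instigator hears the close sound and the fuse audio fades; others hear the distant sound.
; A separate function, uncalled here, force-detaches the target pawn from its remote-controlled pawn.
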(event EventBeginPlay
  ;[1 1 1] Begin Play
  (Networking|SwitchHasAuthority
    (:Authority
      (Utilities|FlowControl|Delay 0.0)
      (Delivery|DoExplosion))
    (:Remote
      (bind _asfort_player_pawn (Utilities|Casting|CastToFortPlayerPawn (Variables|Actor|GetInstigator))))))

(event EventOnExploded
  (params (HitActors :type Array%20of%20Actor%20Object%20References) (HitResults :type Array%20of%20Hit%20Result%20Structures))
  ;[1 1 1] OnExploded - Emitter, Sound, Camera Shake
  (Networking|CanExecuteCosmeticEvents
    (:True
      (if (Pawn|IsLocallyControlled (Game|GetInstigator))
        (Audio|PlaySoundatLocation (Variables|Sounds|GetCue_CloseSound) (Transformation|GetActorLocation) "0, 0, 0" 0.6) ; Play Close Explosion
        (Audio|Components|Audio|FadeOut (Variables|B_Prj_Grenade|GetGrenadeFuse_AudioComponent) 0.2 0.0 "Linear" :!class /Script/Engine.AudioComponent) ; Fade Out Fuse Sound
        (else
          (Audio|PlaySoundatLocation (Variables|Sounds|GetCue_DistantSound) (Transformation|GetActorLocation) "0, 0, 0" 0.4) ; Play Distant Explosion
        )))))

(fn HandlePawnDetachRC
  (bind _asfort_player_pawn (Utilities|Casting|CastToFortPlayerPawn (Variables|BPrjGrenade|GetTarget)))
  (RemoteControlPawn|ForceDetachfromRemoteControlledPawn _asfort_player_pawn))

(fn ConstructionScript
  (Networking|CanExecuteCosmeticEvents
    (:True
      (Components|Movement|SetUpdatedComponent (Variables|Default|GetRotatingMovement) 0))))
