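# Unlit, translucent, two-sided Niagara mesh-particle material for the idle ribbon effect of HighTower Date chain lightning.
# A curve atlas band slides along UV1 with looping dynamic-parameter Time, multiplied by sine UV masks and panning directional noise.
# Dynamic Erode cuts this mask, then particle alpha, depth fade and two-sided sign shape opacity.
# Particle color feeds emissive adjustments, and SplineThicken sets strand thickness from the dynamic Width.
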Material /HighTower/Items/Date/ChainLightning/FX/M_HighTower_Date_Idle_Ribbon
graph {
 properties {
  bAutomaticallySetUsageInEditor=false
  bUsedWithNiagaraMeshParticles=true
  blendMode=BLEND_Translucent
  shadingModel=MSM_Unlit
  twoSided=true
  usedShadingModels=Unlit
 }
 nodes [
  n1 = MaterialFunctionCall(UV Channel=MaterialFunctionCall(XY=TextureCoordinate() {coordinateIndex=1}) {materialFunction=&Swizzle}.YX) {materialFunction=&LinearGradient}
  n2 = ScalarParameter() {parameterName=TimeOffset}
  n3 = ScalarParameter() {defaultValue=0.2, parameterName=Param_4}
  n4 = Frac(None=Add(A=DynamicParameter() {paramNames [Erode,Width,IsUltra,Time]}.Time, B=Add(A=@n2, B=@n3) {constB=0.33}))
  n5 = CurveAtlasRowParameter(CurveTime=Add(A=Divide(A=@n1.UGradient) {constB=2.5}, B=LinearInterpolate(Alpha=@n4) {constA=-1, constB=0.75})) {
   atlas=/Game/Effects/Fort_Effects/GameplayPulginsAssets/Demon/CA_HighTower_Date
   curve=/Game/Effects/Fort_Effects/GameplayPulginsAssets/Demon/Curve_HighTower_Date_Ball
   parameterName=Param_2
  }
  n6 = MaterialFunctionCall(UV Channel=MaterialFunctionCall(XY=TextureCoordinate()) {materialFunction=&Swizzle}.YX) {materialFunction=&LinearGradient}
  n7 = Multiply(A=@n6.UGradient) {constB=0.5}
  n8 = Saturate(None=Multiply(A=Sine(None=@n7)) {constB=2})
  n9 = Saturate(None=Sine(None=Multiply(A=@n6.VGradient) {constB=0.5}))
  n10 = Multiply(
   A=MaterialFunctionCall(XY=TextureCoordinate()) {materialFunction=&Swizzle}.YX
   B=AppendVector(A=ScalarParameter() {defaultValue=1, parameterName=Param_1}, B=ScalarParameter() {defaultValue=0.5, parameterName=Param_2})
  )
  n11 = AppendVector(A=ScalarParameter() {defaultValue=1, parameterName=Param_6}, B=ScalarParameter() {defaultValue=0.1, parameterName=Param_7})
  n12 = TextureSample(UVs=Divide(A=Panner(Coordinate=@n10, Speed=@n11) {speedX=3, speedY=1}) {constB=1.5}) {samplerType=SAMPLERTYPE_LinearGrayscale, texture=/Game/Effects/Fort_Effects/Emotes/UltraLight/Texture/T_DirectionalNoise_Greyscale}
  n13 = ScalarParameter() {defaultValue=1, parameterName=Param_3}
  n14 = Multiply(A=Multiply(A=@n5.A), B=Multiply(A=Multiply(A=@n8, B=@n9), B=Multiply(A=@n12.R, B=@n13) {constB=3}))
  n15 = DynamicParameter() {paramNames [Erode,Width,IsUltra,Time]}
  n16 = ParticleColor()
  n17 = Multiply(A=Saturate(None=Subtract(A=@n14, B=@n15.Erode)), B=@n16.A)
  LinearInterpolate(A=DepthFade(Opacity=LinearInterpolate(Alpha=@n17)) {fadeDistanceDefault=10}, Alpha=TwoSidedSign()) {constB=0}
  MaterialFunctionCall(Emissive=Multiply(A=@n16.RGB)) {materialFunction=&MF_FX_EmissiveAdjustments}
  n18 = Multiply(A=ScalarParameter() {defaultValue=5, parameterName=SplineWidth}, B=@n15.Width)
  MaterialFunctionCall(WidthBase=@n18, WidthTip=@n18) {materialFunction=&SplineThicken}
  Time()
 ]
}
