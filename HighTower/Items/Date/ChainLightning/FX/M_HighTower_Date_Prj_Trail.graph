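# Unlit translucent Niagara ribbon material for the projectile trail of HighTower Date's ChainLightning item.
# Warped UV1 pans over smoke and noise textures, offset by the U_Offset dynamic parameter.
# The noise erodes a tapered sine gradient, trimmed by a sphere mask, to form the trail mask.
# That mask drives opacity and a dark-to-bright green ramp, which IsUltra swaps to orange, tinted by particle color.
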Material /HighTower/Items/Date/ChainLightning/FX/M_HighTower_Date_Prj_Trail
graph {
 properties {bUsedWithNiagaraRibbons=true, blendMode=BLEND_Translucent, pixelDepthOffsetMode=PDOM_Legacy, shadingModel=MSM_Unlit, usedShadingModels=Unlit}
 nodes [
  n1 = MaterialFunctionCall() {materialFunction=&LinearGradient}
  n2 = Multiply(A=@n1.VGradient) {constB=0.5}
  n3 = Saturate(None=Multiply(A=Sine(None=@n2), B=OneMinus(None=@n1.UGradient)))
  n4 = MaterialFunctionCall(Float2=TextureCoordinate() {coordinateIndex=1}) {materialFunction=&BreakOutFloat2Components}
  n5 = Power(Base=@n4.R) {constExponent=0.5}
  n6 = DynamicParameter() {defaultValue=<rgba 1 1 0 1>, paramNames [Erode,U_Offset,IsUltra,Param4]}
  n7 = Panner(
   Coordinate=Multiply(A=AppendVector(A=@n5, B=@n4.G), B=Constant2Vector() {g=0.2, r=1})
   Time=Add(A=Time(), B=@n6.U_Offset)
   Speed=AppendVector(A=ScalarParameter() {defaultValue=-1, parameterName=X_PanSpeed}, B=Constant())
  ) {speedX=-1}
  n8 = TextureSample(UVs=@n7) {texture=&T_Inky_Smoke_Tile}
  n9 = Divide(A=TextureSample(UVs=Multiply(A=@n7) {constB=2}) {samplerType=SAMPLERTYPE_LinearGrayscale, texture=/Game/Effects/Fort_Effects/Emotes/UltraLight/Texture/T_DirectionalNoise_Greyscale}.R) {constB=3}
  n10 = Subtract(A=OneMinus(None=Add(A=@n8.R, B=@n9))) {constB=0.5}
  n11 = Multiply(
   A=SphereMask(A=TextureCoordinate(), B=Constant2Vector() {g=0.5, r=2}) {attenuationRadius=2, hardnessPercent=97}
   B=Saturate(None=Subtract(A=@n3, B=@n10))
  ) {constA=1}
  n12 = LinearInterpolate(A=Constant3Vector() {constant=<rgba 0.0005 0.0087 0.001 1>}, B=Constant3Vector() {constant=<rgba 0.3176 0.8316 0.0474 1>}, Alpha=@n11)
  n13 = LinearInterpolate(A=Constant3Vector() {constant=<rgba 0.4844 0.0528 0 1>}, B=Constant3Vector() {constant=<rgba 1 0.5237 0.0335 1>}, Alpha=@n11)
  n14 = DynamicParameter() {paramNames [Erode,U_Offset,IsUltra,Param4]}
  n15 = ParticleColor()
  MaterialFunctionCall(Emissive=Multiply(A=LinearInterpolate(A=@n12, B=@n13, Alpha=@n14.B), B=@n15.RGB)) {materialFunction=&MF_FX_EmissiveAdjustments}
  n16 = ScalarParameter() {defaultValue=3, parameterName=Opacity}
  n17 = LinearInterpolate(Alpha=MaterialFunctionCall() {materialFunction=&LinearGradient}.UGradient) {constA=0.5}
  Multiply(A=@n15.A, B=Multiply(A=@n11, B=Multiply(A=@n16, B=@n17) {constA=1}))
  Multiply(A=@n10, B=@n6.Erode)
  Multiply()
  TextureSample(UVs=@n7) {samplerType=SAMPLERTYPE_LinearGrayscale, texture=/Game/Global/Textures/Default/Tileable_Patterns/T_StormCloudDetail01_M}
 ]
}
