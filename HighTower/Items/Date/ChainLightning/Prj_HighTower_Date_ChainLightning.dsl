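; HighTower Date projectile that explodes on hit, stop or an authority-set lifetime timer, then hides and expires.
; DoOnce gates and a played flag limit explosion cosmetics, with separate water and impact gameplay cues.
; ChainLightning sphere-traces for unhit pawns in line of sight, applies a gameplay effect, fires an arc cue and returns the last target.
; Neither it nor FindChainLightningTargets, which gathers nearby pawns, is called within this graph.
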(event EventBeginPlay
  (bind _row_lifetimebeforeautoexplode (Variables|Default|GetRow_LifetimeBeforeAutoExplode))
  (bind _returnvalue (ScalableFloat|GetValueatLevel _row_lifetimebeforeautoexplode))
  ;[0.02277 1 0.047 0.3] BeginPlay()
  ; Add streched mesh between the muzzleflash and porjectile to brindge gap
  ; For remote players, disconnect cosmetics and add lerping
  (|CallParentFunction)
  (Utilities|FlowControl|Sequence
    (:then_0
      (Variables|Default|SetFortAbilitySystemComponent (Ability|GetFortAbilitySystemComponent (Game|GetInstigator))))
    (:then_1
      (Networking|SwitchHasAuthority
        (:Authority
          (if (> _returnvalue 0)
            (bind _returnvalue_2 (Utilities|Time|SetTimerbyEvent (EventDispatchers|CreateEvent :!event Custom|TryExplode) _returnvalue))
            (Variables|Default|SetTimerHandle_AutoExplode _returnvalue_2)))))))

(event Custom|ExplodeWhenHitAndStop
  (bind (_bblockinghit _binitialoverlap _time _distance _location _impactpoint _normal _impactnormal _physmat _hitactor) (Collision|BreakHitResult (Variables|Default|GetHit)))
  ;[1 1 1] Try Explode
  (Variables|Default|SetIsOnHitExplosion true)
  (Utilities|FlowControl|DoOnce false
    (:Completed
      (Delivery|DoExplosionatLocation (Transformation|GetActorLocation))
      (Utilities|Time|ClearandInvalidateTimerbyHandle (Variables|Default|GetTimerHandle_AutoExplode))
      (Networking|CanExecuteCosmeticEvents
        (:True
          ; Impact GC - Call it here because we need the hit normal to set the impact FX facing correctly
          (if (not (Variables|Default|GetHasPlayedExplosionCosmetics))
            (GameplayCue|ExecuteGameplayCueOnActor(Burst) self (Variables|Default|GetImpactGC_Tag) (Ability|GameplayCue|MakeGameplayCueParameters :NormalizedMagnitude 0.0 :RawMagnitude (CallFunction|GetImpactGCMagnitude) :Location _location :Normal _normal :EffectCauser _hitactor :GameplayEffectLevel 1 :AbilityLevel 1 :bReplicateLocationWhenUsingMinimalRepProxy false))
            (Variables|Default|SetHasPlayedExplosionCosmetics true)))))))

(event Custom|ExplodeInAir
  (bind (_bblockinghit _binitialoverlap _time _distance _location _impactpoint _normal _impactnormal _physmat _hitactor) (Collision|BreakHitResult (Variables|Default|GetHit)))
  (Networking|CanExecuteCosmeticEvents
    (:True
      ; Impact GC - Call it here because we need the hit normal to set the impact FX facing correctly
      (if (not (Variables|Default|GetHasPlayedExplosionCosmetics))
        (GameplayCue|ExecuteGameplayCueOnActor(Burst) self (Variables|Default|GetImpactGC_Tag) (Ability|GameplayCue|MakeGameplayCueParameters :NormalizedMagnitude 0.0 :RawMagnitude (CallFunction|GetImpactGCMagnitude) :Location _location :Normal _normal :EffectCauser _hitactor :GameplayEffectLevel 1 :AbilityLevel 1 :bReplicateLocationWhenUsingMinimalRepProxy false))
        (Variables|Default|SetHasPlayedExplosionCosmetics true)))))

(event Custom|TryExplode
  (Variables|Default|SetIsOnHitExplosion false)
  (Utilities|FlowControl|DoOnce false
    (:Completed
      (Delivery|DoExplosionatLocation (Transformation|GetActorLocation))
      (CallFunction|ExplodeinAir))))

(event EventOnStop
  (params (Hit :type Hit%20Result%20Structure%20%28by%20ref%29))
  (bind (_bblockinghit _binitialoverlap _time _distance _location _impactpoint _normal _impactnormal _physmat _hitactor) (Collision|BreakHitResult (Variables|Default|GetHit)))
  ;[0.8594 0.3941 0 0.5] On Stop
  ; Explode when the projectile stopped, this is necessary because Event Hit() does not get called on the server.
  (Variables|Default|SetHit Hit)
  ;[0.3222 1 0.9701] On Hit or enter water do the Explode, Also set the rep notify to hide the actor / turn off collision. Set the lifespan to kill the projectile
  (Utilities|Casting|CastToWaterBody _hitactor
    (:then
      (CallFunction|ExplosiononWater) ; Must be called before ExplodeWhenHitAndStop so that the proper cosmetics are used.
    )
    (:CastFailed
      _))
  (CallFunction|ExplodewhenHitandStop))

(event Collision|EventHit
  (params (MyComp :type Primitive%20Component%20Object%20Reference) (Other :type Actor%20Object%20Reference) (OtherComp :type Primitive%20Component%20Object%20Reference) (bSelfMoved :type Boolean) (HitLocation :type Vector :default "0, 0, 0") (HitNormal :type Vector :default "0, 0, 0") (NormalImpulse :type Vector :default "0, 0, 0") (Hit :type Hit%20Result%20Structure%20%28by%20ref%29))
  (bind (_bblockinghit _binitialoverlap _time _distance _location _impactpoint _normal _impactnormal _physmat _hitactor) (Collision|BreakHitResult (Variables|Default|GetHit)))
  ;[0.3222 1 0.9701] On Hit or enter water do the Explode, Also set the rep notify to hide the actor / turn off collision. Set the lifespan to kill the projectile
  (Variables|Default|SetHit Hit)
  (Utilities|Casting|CastToWaterBody _hitactor
    (:then
      (CallFunction|ExplosiononWater) ; Must be called before ExplodeWhenHitAndStop so that the proper cosmetics are used.
    )
    (:CastFailed
      _))
  (CallFunction|ExplodewhenHitandStop))

(event Custom|ExplosionOnWater
  (Utilities|FlowControl|DoOnce false
    (:Completed
      (Networking|CanExecuteCosmeticEvents
        (:True
          (if (not (Variables|Default|GetHasPlayedExplosionCosmetics))
            (GameplayCue|ExecuteGameplayCue%28Burst%29Local self (Variables|Default|GetWaterImpactGC_Tag) (Ability|GameplayCue|MakeGameplayCueParameters :NormalizedMagnitude 0.0 :RawMagnitude 0.0 :Location (Transformation|GetWorldLocation (Variables|Transformation|GetRootComponent)) :Normal "0,0,0" :GameplayEffectLevel 1 :AbilityLevel 1 :bReplicateLocationWhenUsingMinimalRepProxy false)) ; WaterImpact
            (Variables|Default|SetHasPlayedExplosionCosmetics true)
            (Rendering|SetActorHiddenInGame true)))))))

(event EventOnExploded
  (params (HitActors :type Array%20of%20Actor%20Object%20References) (HitResults :type Array%20of%20Hit%20Result%20Structures))
  (Variables|Default|SetHideAndKill true)
  (Actor|SetLifeSpan 3.0))

(event Custom|GameplayCueMulticast
  (params (Target :type Actor%20Object%20Reference) (GameplayCueTag :type Gameplay%20Tag%20Structure) (CueParameters :type Gameplay%20Cue%20Parameters%20Structure%20%28by%20ref%29))
  (GameplayCue|ExecuteGameplayCue%28Burst%29Local Target GameplayCueTag CueParameters))

(fn ChainLightning
  (params (InputPin :type Actor%20Object%20Reference))
  (bind _returnvalue (Transformation|GetActorLocation InputPin))
  (Variables|Default|SetFoundChainTarget 0)
  (Utilities|IsValid InputPin
    (:Is%20Valid
      (bind _outhits (Collision|MultiSphereTraceByChannel _returnvalue (+ _returnvalue 1.0) (ScalableFloat|GetValueatLevel (Variables|Default|GetChainLightningRange))))
      (bind _array_element (Utilities|Array|ForEachLoopwithBreak _outhits)
        (:LoopBody
          (bind _asfort_player_pawn (Utilities|Casting|CastToFortPlayerPawn (.HitActor (Collision|BreakHitResult _array_element))))
          (if (== (Utilities|Array|FindItem (Variables|Default|GetHitTargets) _asfort_player_pawn) -1)
            ;[1 1 1] Check LOS
            (bind _in_line_of_sight (Class|MissionBlueprintFunctionLibrary|CheckBasicActorLineOfSight InputPin "0, 0, 0" _asfort_player_pawn "0, 0, 0" true "TraceTypeQuery1" "None" 0.0 false true (Utilities|Array|MakeArray InputPin)))
            (if _in_line_of_sight
              (Utilities|Array|AddUnique (Variables|Default|GetHitTargets) _asfort_player_pawn)
              (if (Utilities|NotEqual(Object) _asfort_player_pawn (Variables|Actor|GetInstigator))
                ;[1 1 1] Apply Damage
                (GameplayEffects|FortApplyGameplayEffectToTarget :self (Variables|Default|GetFortAbilitySystemComponent) :GameplayEffectClass (Variables|Default|GetGE_Targeted) :TargetOfEffect (Ability|GetFortAbilitySystemComponent _asfort_player_pawn) :Level 0.0)
                ;[1 1 1] Chain Lightning GC
                (CallFunction|GameplayCueMulticast (Game|GetInstigator) (Variables|Default|GetGC_ChainLightning) (Ability|GameplayCue|MakeGameplayCueParameters :NormalizedMagnitude 0.0 :RawMagnitude 0.0 :Location _returnvalue :Normal (Transformation|GetActorLocation _asfort_player_pawn) :Instigator InputPin :EffectCauser _asfort_player_pawn :GameplayEffectLevel 1 :AbilityLevel 1 :bReplicateLocationWhenUsingMinimalRepProxy false))
                (Variables|Default|SetFoundChainTarget _asfort_player_pawn)))))
        (:Completed
          _)))
    (:Is%20Not%20Valid
      _))
  (return (Variables|Default|GetFoundChainTarget)))

(fn FindChainLightningTargets
  (params (HitActor :type Actor%20Object%20Reference))
  (bind _chainlightningtargets (Variables|Default|GetChainLightningTargets))
  (Utilities|IsValid HitActor
    (:Is%20Valid
      (bind _asfort_player_pawn (Utilities|Casting|CastToFortPlayerPawn HitActor)
        (:then
          ;[1 1 1] Get all pawns near hit loc and put them in an array
          (Utilities|Array|AddUnique (Variables|Default|GetChainLightningTargets) _asfort_player_pawn))
        (:CastFailed
          _))
      (bind _outhits (Collision|MultiSphereTraceByChannel (Transformation|GetActorLocation) (+ (Transformation|GetActorLocation) 1.0) 512.0 "TraceTypeQuery1" false (Utilities|Array|MakeArray (Variables|Actor|GetInstigator))))
      (for _array_element _outhits
        (bind _asfort_player_pawn_2 (Utilities|Casting|CastToFortPlayerPawn (.HitActor (Collision|BreakHitResult _array_element))))
        (if (not (Utilities|Array|ContainsItem _chainlightningtargets _asfort_player_pawn_2))
          (Utilities|Array|AddUnique _chainlightningtargets _asfort_player_pawn_2))))
    (:Is%20Not%20Valid
      _))
  (return))

(fn GetImpactGCMagnitude
  (return (Math|Conversions|ToFloat(Boolean) (Variables|Default|GetIsOnHitExplosion))))

(fn OnRep_HideAndKill
  ;[1 1 1] If we set hide and kill to true disable collision set the actor to hidden in game and fade out.
  (if (Variables|Default|GetHideAndKill)
    (Rendering|SetActorHiddenInGame true)
    (Collision|SetActorEnableCollision false)
    (Audio|Components|Audio|FadeOut (Variables|Default|GetAudio) 0.2 0.0 "Linear" :!class /Script/Engine.AudioComponent)))

(fn ConstructionScript)
